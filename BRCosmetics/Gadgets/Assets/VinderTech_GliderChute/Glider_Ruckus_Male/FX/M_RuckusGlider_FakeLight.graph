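# Fake light material for Niagara sprites on the Ruckus glider FX, tinting the scene behind each particle without real lighting.
# Unlit Modulate blend with depth test disabled measures the background world position's distance from particle center, normalized by stretched radius.
# The inverted distance, raised to the fourth power, blends from neutral 1 to an HDR color whose hue shifts with distance.
# Particle Stretch also scales a vertex-from-center offset, apparently enlarging the sprite.
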Material /BRCosmetics/Gadgets/Assets/VinderTech_GliderChute/Glider_Ruckus_Male/FX/M_RuckusGlider_FakeLight
graph {
 properties {
  bDisableDepthTest=true
  bUsedWithNiagaraSprites=true
  blendMode=BLEND_Modulate
  pixelDepthOffsetMode=PDOM_Legacy
  shadingModel=MSM_Unlit
  translucencyPass=MTP_BeforeDOF
  usedShadingModels=Unlit
 }
 nodes [
  n1 = VectorParameter() {parameterName="Sun Color && Hue Shift", defaultValue=<rgba 18.7128 66.3398 150 0.03>}
  n2 = MaterialFunctionCall() {materialFunction=&WorldPositionBehindTranslucency}
  n3 = ParticlePositionWS()
  n4 = ParticleRadius()
  n5 = ScalarParameter() {defaultValue=1, parameterName=Particle Stretch}
  n6 = Add(A=@n5)
  n7 = Saturate(None=Divide(A=Distance(A=@n2.Result, B=@n3), B=Multiply(A=@n4, B=@n6) {constB=2}))
  n8 = OneMinus(None=@n7)
  n9 = Multiply(A=@n8, B=@n8)
  LinearInterpolate(
   B=MaterialFunctionCall(Hue Shift Percentage=Multiply(A=@n1.A, B=@n7), Texture=@n1.RGB) {materialFunction=&HueShift}.Result
   Alpha=Multiply(A=@n9, B=@n9)
  ) {constA=1}
  n10 = WorldPosition()
  Reroute(None=Multiply(A=Subtract(A=@n10.XYZ, B=@n3), B=@n5))
 ]
}
